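# Conversation database for the PaperBlaze NPC, holding the greeting speech that opens a player's talk with this story-guarding character.
# The entry tag leads to one Speech node that picks a line by requirement and priority: first-ever meeting, reactions to specific worn outfits, or lines gated by match conversation count.
# The speech then branches through link nodes to the shared quest and basic-service conversations and to PaperBlaze's own Dialogue entry.
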ConversationDatabase /NPCLibrary/NPCs/PaperBlaze/NPCLibrary_Conversation_PaperBlaze
entry_points [{tag=Conversation.Entry.PaperBlaze, destinations [#1]}]
nodes [
 ConversationEntryPointNode_7 {
  id=#1
  class=ConversationEntryPointNode
  outputs [#2]
  properties {entryTag {tagName=Conversation.Entry.PaperBlaze}}
 }
 Speech_C_4 {
  id=#2
  class=Speech_C
  outputs [#3, #4, #5]
  properties {
   nodeName=""
   generalConfig {
    defaultMessage="Stories must be protected, from the ravages of time and from those who would destroy them."
    contextualMessages [
     {
      message="May the moonlight protect your path, stranger."
      contextRequirements [{requirement {class=FortControllerRequirement_IsFirstEverConversationWithNPC}}]
      priority {curve {rowName=Tandem.Speech.Priority.FirstEverInteraction}}
     }
     {
      message="Go ahead, protect your stories with your heart. I will protect mine with my blade."
      contextRequirements [{requirement {class=FortControllerRequirement_HasCID, properties {allowedCIds [/BRCosmetics/Athena/Items/Cosmetics/Characters/Character_PaperGlow]}}}]
      priority {curve {rowName=Tandem.Speech.Priority.OutfitReaction}}
     }
     {
      message="You fear nothing, king of the gods? Hm. Nothing but a story, perhaps."
      contextRequirements [{requirement {
         class=FortControllerRequirement_HasCID
         properties {allowedCIds [&Character_StunningMountain_Armor]}
        }}]
      priority {curve {rowName=Tandem.Speech.Priority.OutfitReaction}}
     }
     {
      message=Prophecies are merely stories. Hearing them shapes us - and fate itself.
      contextRequirements [{requirement {class=FortControllerRequirement_HasCID, properties {allowedCIds [/BRCosmetics/Athena/Items/Cosmetics/Characters/Character_DaffodilSong]}}}]
      priority {curve {rowName=Tandem.Speech.Priority.OutfitReaction}}
     }
     {
      message=You stand at the precipice of a grand new story. Its ending relies on you.
      contextRequirements [{requirement {
         class=FortControllerRequirement_HasCID
         properties {allowedCIds [&Character_VoidRedemption_Rebel]}
        }}]
      priority {curve {rowName=Tandem.Speech.Priority.OutfitReaction}}
     }
     {
      message="Stories must be protected, from the ravages of time and from those who would destroy them."
      contextRequirements [{requirement {class=FortControllerRequirement_CheckNPCMatchConversationCount, properties {nPCUniqueIdTag {tagName=AISpawnerData.Type.Tandem.PaperBlaze}}}}]
      priority {curve {rowName=None}}
     }
     {
      message="Stories can end empires, forge communities, shape civilizations. As long as we listen."
      contextRequirements [{requirement {class=FortControllerRequirement_CheckNPCMatchConversationCount, properties {nPCUniqueIdTag {tagName=AISpawnerData.Type.Tandem.PaperBlaze}}}}]
      priority {curve {rowName=None}}
     }
     {
      message="There are few weapons as powerful as the right story, stranger."
      contextRequirements [{requirement {class=FortControllerRequirement_CheckNPCMatchConversationCount, properties {nPCUniqueIdTag {tagName=AISpawnerData.Type.Tandem.PaperBlaze}}}}]
      priority {curve {rowName=None}}
     }
     {
      message=Every story has a storyteller. And storytellers can lie.
      contextRequirements [{requirement {class=FortControllerRequirement_CheckNPCMatchConversationCount, properties {nPCUniqueIdTag {tagName=AISpawnerData.Type.Tandem.PaperBlaze}}}}]
      priority {curve {rowName=None}}
     }
     {
      message=There is no one truth to any story. Remember that as you listen.
      contextRequirements [{requirement {class=FortControllerRequirement_CheckNPCMatchConversationCount, properties {nPCUniqueIdTag {tagName=AISpawnerData.Type.Tandem.PaperBlaze}}}}]
      priority {curve {rowName=None}}
     }
     {
      message="You carry many stories, as does everyone. And they are all precious."
      contextRequirements [{requirement {class=FortControllerRequirement_CheckNPCMatchConversationCount, properties {nPCUniqueIdTag {tagName=AISpawnerData.Type.Tandem.PaperBlaze}}}}]
      priority {curve {rowName=None}}
     }
    ]
   }
  }
 }
 ConversationLinkNode_2 {
  id=#5
  class=ConversationLinkNode
  properties {remoteEntryTag {tagName=Conversation.Entry.PaperBlaze.Dialogue}}
 }
 ConversationLinkNode_0 {id=#4, class=ConversationLinkNode, properties {remoteEntryTag {tagName=Conversation.Entry.BasicServices}}}
 ConversationLinkNode_1 {id=#3, class=ConversationLinkNode, properties {remoteEntryTag {tagName=Conversation.Entry.NPCLibrary.Quests}}}
]
